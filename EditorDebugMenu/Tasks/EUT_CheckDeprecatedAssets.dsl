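; Editor utility task that reports which packages still reference deprecated assets, for the editor debug menu.
; It reads soft asset references from a data table and queries the Asset Registry for each one's package referencers.
; Referencer names containing any ignore-list substring are left out of the printed list, though the reported count includes them.
; The task's notification text mentions counting level actors, which does not match this behavior.
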(event Task|EventBeginExecution
  ;[1 1 1] Begin
  (Task|SetTaskNotificationText (Utilities|Text|MakeLiteralText "NSLOCTEXT(\"[04CA21917600573B27D84B72F603FD43]\", \"321F5698401812E72148F3A9ADDC8FA6\", \"Counting Actors in level...\")"))
  (Development|PrintText 0 "INVTEXT(\"--- Running Asset Dependency Checker ---\")")
  (CallFunction|CheckAssets)
  (Development|PrintText 0 "INVTEXT(\"---  Asset Dependency Checker Completed ---\")")
  (CallFunction|CompleteTask))

(event Custom|Check%20Assets
  (bind _assets_to_check (Variables|Setup|GetAssets%20to%20check))
  (bind _outrownames (DataTable|GetDataTableRowNames _assets_to_check))
  (for _array_element _outrownames
    (bind _returnvalue (Utilities|GetDataTableRow _assets_to_check _array_element))
    (CallFunction|CheckAsset (Utilities|Struct|BreakStructDeprecatedAssetsCheck _returnvalue))))

(event Task|EventCancelRequested
  ;[1 1 1] Cancel
  (CallFunction|CompleteTask))

(event Custom|Complete%20Task
  ;[1 1 1] Finish
  (Task|FinishExecutingTask true))

(fn CheckAsset
  (params (Asset :type Soft%20Object%20Reference))
  (bind _returnvalue (Utilities|ToSoftObjectPath Asset))
  (bind _pathstring (Utilities|BreakSoftObjectPath _returnvalue))
  (bind (_lefts _rights) (Utilities|String|Split _pathstring "."))
  (LocalVariables|SetCurrent%20Package%20Name _rights)
  (bind (_outreferencers _returnvalue_2) (AssetRegistry|GetReferencers (AssetRegistry|GetAssetRegistry) (Utilities|String|StringToName _lefts) (Utilities|Struct|MakeAssetRegistryDependencyOptions true)))
  (if _returnvalue_2
    (for _array_element _outreferencers
      (if (CallFunction|CheckforIgnoredNames _array_element)
        (LocalVariables|SetOutput%20String (Utilities|Text|FormatText "NSLOCTEXT(\"[04CA21917600573B27D84B72F603FD43]\", \"010B6E954B812A9E70AEE59BF0AD3B1E\", \"{output}\\r\\n--{new}\")" (LocalVariables|GetOutput%20String) (Utilities|Text|ToText(Name) _array_element)))))
    (Development|PrintText 0 (Utilities|Text|FormatText "NSLOCTEXT(\"[04CA21917600573B27D84B72F603FD43]\", \"228FF7DC4850D258C1CE59974B7CF215\", \"-Asset [{package}] found {num} references:\\r\\n{refs}\")" (LocalVariables|GetCurrent%20Package%20Name) (Utilities|Array|Length _outreferencers) (LocalVariables|GetOutput%20String)) false)))

(fn CheckForIgnoredNames
  (params (InName :type Name))
  (for _array_element (Variables|Default|GetNames%20to%20ignore)
    (if (Utilities|String|Contains (Utilities|String|ToString(Name) InName) _array_element)
      (return false)))
  (return true))
