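; Editor utility task in the editor debug menu that converts Building Level Instances to Level Instances.
; On begin it sets the task notification text, calls ConvertLIs (whose body is not in this file), then calls Complete Task.
; A cancel request also calls Complete Task, which finishes the executing task with a true flag in both cases.
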(event Task|EventBeginExecution
  ;[1 1 1] Begin
  (Task|SetTaskNotificationText (Utilities|Text|MakeLiteralText "NSLOCTEXT(\"[4E8C887C4D2371BB999C715A333A90BD]\", \"5CA576FC4EE0865D4AA9619ED3DCD80B\", \"Converting Building Level Instances to Level Instances...\")"))
  (CallFunction|ConvertLIs)
  (CallFunction|CompleteTask))

(event Task|EventCancelRequested
  ;[1 1 1] Cancel
  (CallFunction|CompleteTask))

(event Custom|Complete%20Task
  ;[1 1 1] Finish
  (Task|FinishExecutingTask true))
