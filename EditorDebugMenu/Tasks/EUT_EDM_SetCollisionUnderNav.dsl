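; Editor utility task for a debug menu that enables "fill collision underneath for navmesh" on static mesh components of the selected actors.
; On begin it sets the task notification, runs SetNav, then finishes; cancelling also finishes.
; SetNav loops each selected actor's static mesh components, sets the flag through editor property access, and counts those not already enabled.
; It prints how many components changed.
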(event Task|EventBeginExecution
  ;[1 1 1] Begin
  (Task|SetTaskNotificationText (Utilities|Text|MakeLiteralText "NSLOCTEXT(\"[14F3D79FE4F9B68803538764C658C7ED]\", \"F15B46A5432468D6C1A1E3A40832015E\", \"Setting Nav\")"))
  (CallFunction|SetNav)
  (CallFunction|CompleteTask))

(event Task|EventCancelRequested
  ;[1 1 1] Cancel
  (CallFunction|CompleteTask))

(event Custom|Complete%20Task
  ;[1 1 1] Finish
  (Task|FinishExecutingTask true))

(fn SetNav
  (bind _returnvalue (Scouting|GetSelectedActors))
  (for _array_element _returnvalue
    (bind _asactor (Utilities|Casting|CastToActor _array_element))
    (for _array_element_2 (Actor|GetComponentsByClass _asactor "/Script/Engine.StaticMeshComponent")
      (bind _propertyvalue (Utilities|GetEditorProperty _array_element_2 "bFillCollisionUnderneathForNavmesh"))
      (LocalVariables|SetOldValue (Math|Boolean|MakeLiteralBool _propertyvalue))
      (bind _returnvalue_2 (Utilities|SetEditorProperty _array_element_2 "bFillCollisionUnderneathForNavmesh" true))
      (if _returnvalue_2
        (if (!= (LocalVariables|GetOldValue) true)
          (Math|Integer|IncrementInt (LocalVariables|GetAcc))))))
  (Development|PrintText 0 (Utilities|Text|FormatText "NSLOCTEXT(\"[14F3D79FE4F9B68803538764C658C7ED]\", \"EE6107404A2C4FBA75BA32BBEF4085A0\", \"{num} components changed\")" (LocalVariables|GetAcc))))
