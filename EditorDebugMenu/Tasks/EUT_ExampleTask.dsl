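; Example Editor Utility Task that counts the actors in the level from an editor debug menu.
; On begin it sets the task notification text and starts a short one-shot timer, whose event gets all Actors and prints the total.
; A cancel request clears the stored timer handle.
; Both the timer and cancel paths route through a shared CompleteTask event that finishes the task as successful.
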(event Task|EventBeginExecution
  ;[1 1 1] Begin
  (Task|SetTaskNotificationText (Utilities|Text|MakeLiteralText "NSLOCTEXT(\"[426EA78751F78E3C706392B4993745B6]\", \"88A53BFD48DF3CF6459BCDB5910C5DC5\", \"Counting Actors in level...\")"))
  (bind _returnvalue (Utilities|Time|SetTimerbyEvent (delegate Custom|Timed%20Event) 1.0))
  (Variables|Default|SetTimer _returnvalue))

(event Custom|Timed%20Event
  (bind _outactors (Actor|GetAllActorsOfClass 0 "/Script/Engine.Actor"))
  (Development|PrintText 0 (Utilities|Text|FormatText "NSLOCTEXT(\"[426EA78751F78E3C706392B4993745B6]\", \"2C7E0AD64875CB37C5392D9A5668D9D1\", \"[EUT Example] ------------ There are {num} actors in the level -------------\")" (Utilities|Array|Length _outactors)))
  (CallFunction|CompleteTask))

(event Task|EventCancelRequested
  ;[1 1 1] Cancel
  (Utilities|Time|ClearandInvalidateTimerbyHandle 0 (Variables|Default|GetTimer))
  (CallFunction|CompleteTask))

(event Custom|Complete%20Task
  ;[1 1 1] Finish
  (Task|FinishExecutingTask true))
